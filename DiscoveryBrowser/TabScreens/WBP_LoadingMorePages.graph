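# WBP_LoadingMorePages is a Discovery Browser UserWidget that shows a "Loading more" indicator, presumably while further pages load.
# A SafeZone without top padding holds a hit-test-invisible Overlay, with a vertically centered row of UIKit throbber and uppercase CommonTextBlock.
# It contains no working logic: the EventGraph is empty, and the auto-generated function taking a visibility input has only an entry node.
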WidgetBlueprint /DiscoveryBrowser/TabScreens/WBP_LoadingMorePages
graphs [
 EventGraph {links [], nodes []}
 WBP_LoadingMorePages_AutoGenFunc {nodes [WBP_LoadingMorePages_AutoGenFunc!() -> (then, InVisibility:ESlateVisibility Enum=Visible) {is=title K2Node_FunctionEntry}]}
]
variables []
widgets {
 info {inheritedWidgetCount=0, parentClass=&UserWidget, rootWidgetClass=&SafeZone, widgetCount=5}
 widgets [
  SafeZone_Loading {widgetClassPath=&SafeZone, widget=SafeZone_Loading, properties {padTop=false}}
  Overlay_LoadingMorePages {
   widgetClassPath=&Overlay
   bIsVariable=true
   parent=SafeZone_Loading
   widget=Overlay_LoadingMorePages
   properties {visibility=HitTestInvisible}
   slot {class=SafeZoneSlot}
  }
  HB_Loading {
   widgetClassPath=&HorizontalBox
   parent=Overlay_LoadingMorePages
   widget=HB_Loading
   slot {class=OverlaySlot, properties {horizontalAlignment=HAlign_Fill, verticalAlignment=VAlign_Center}}
  }
  WBP_UIKit_Throbber {
   widgetClassPath=/FortUILibrary/UIKit/Throbber/WBP_UIKit_Throbber
   bIsVariable=true
   parent=HB_Loading
   widget=WBP_UIKit_Throbber
   slot {class=HorizontalBoxSlot, properties {padding=<ltrb 0 0 12 0>}}
  }
  Text_LoadingMore {
   widgetClassPath=&CommonTextBlock
   parent=HB_Loading
   widget=Text_LoadingMore
   properties {
    mobileFontSizeMultiplier=1.2
    text=Loading more
    textTransformPolicy=ToUpper
    font {
     bForceMonospaced=false
     bMaterialIsStencil=false
     fontMaterial=null
     fontObject=&HeadingNow
     letterSpacing=0
     monospacedWidth=1
     size=15.75
     skewAmount=0
     typefaceFontName=75 Medium
     outlineSettings {
      bApplyOutlineToDropShadows=false
      bMiteredCorners=false
      bSeparateFillAlpha=false
      outlineMaterial=null
      outlineSize=0
      outlineColor=<rgba 0 0 0 1>
     }
    }
    shadowOffset=<0 0>
   }
   slot {class=HorizontalBoxSlot, properties {verticalAlignment=VAlign_Center}}
  }
 ]
}
